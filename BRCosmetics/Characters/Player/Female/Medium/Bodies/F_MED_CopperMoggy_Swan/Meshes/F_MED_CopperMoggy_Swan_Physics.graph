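# Physics asset for the F_MED_CopperMoggy_Swan female medium player body cosmetic, set up for a dynamic cloak.
# Four dyn_cloak bones form a chain from the MSTR_cloak anchor, linked by constraints with swing and twist limits.
# All bodies use capsules, including unconstrained pelvis, spine and upper-arm volumes around the torso.
# The upper arms are shaped asymmetrically, and MSTR_sleeve has only a tiny unconstrained capsule.
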PhysicsAsset /BRCosmetics/Characters/Player/Female/Medium/Bodies/F_MED_CopperMoggy_Swan/Meshes/F_MED_CopperMoggy_Swan_Physics
bodies {
 MSTR_cloak [{shapeType=Capsule, length=0.505, radius=0.505, shapeName=MSTR_cloak_capsule, rotation=<rot 0 -90 0>}]
 MSTR_sleeve [{shapeType=Capsule, length=0.1, radius=0.1, shapeName=MSTR_sleeve_capsule}]
 dyn_cloak_01 [
  {
   shapeType=Capsule
   length=10.0167
   radius=11.2894
   shapeName=dyn_cloak_01_capsule
   center=<8.7499 -11.8362 -1.9677>
   rotation=<rot -1.3509 -89.9999 -1.409>
  }
  {shapeType=Capsule, length=15, radius=6.9106, shapeName=dyn_cloak_01_capsule, center=<0.6773 -6.5303 -0.3463>}
 ]
 dyn_cloak_02 [{
   shapeType=Capsule
   length=14.6893
   radius=10.5
   shapeName=dyn_cloak_02_capsule
   center=<8.5785 -9.9919 -0.9588>
   rotation=<rot -0.6437 -90 -3.7004>
  }]
 dyn_cloak_03 [{shapeType=Capsule, length=22, radius=8.8, shapeName=dyn_cloak_03_capsule, center=<7.8741 -6.0507 -2.0338>, rotation=<rot -1.4127 -90 4.4847>}]
 dyn_cloak_04 [{
   shapeType=Capsule
   length=18
   radius=9.9
   shapeName=dyn_cloak_04_capsule
   center=<5.5342 -7.5015 -2.2733>
   rotation=<rot -1.3688 -90 14.0429>
  }]
 pelvis [
  {shapeType=Capsule, length=16.5999, radius=12.4463, shapeName=pelvis_capsule, center=<-2.1924 -0.8429 0>}
  {shapeType=Capsule, length=13.4552, radius=12.0516, shapeName=pelvis_capsule, center=<5.5612 -0.1735 0>}
 ]
 spine_01 [{shapeType=Capsule, length=7.9302, radius=8.4073, shapeName=spine_01_capsule, center=<0 0.1289 0>, rotation=<rot 0 -90 0>}]
 spine_02 [{shapeType=Capsule, length=10.8197, radius=7.1804, shapeName=spine_02_capsule, center=<0 1.2355 0>}]
 spine_03 [{shapeType=Capsule, length=12.1043, radius=9.6294, shapeName=spine_03_capsule, center=<1.1832 -0.2998 0>, rotation=<rot 0 -90 0>}]
 spine_04 [{shapeType=Capsule, length=5.6261, radius=12.0592, shapeName=spine_04_capsule, center=<0 -0.0904 0>}]
 spine_05 [
  {
   shapeType=Capsule
   length=10.4267
   radius=11.4723
   shapeName=spine_05_capsule
   center=<2.3396 1.155 1.6718>
   rotation=<rot 0.6308 -90 1.8775>
  }
  {
   shapeType=Capsule
   length=2.4181
   radius=8.8298
   shapeName=spine_05_capsule
   center=<-5.3704 -3.4249 -0.8124>
   rotation=<rot 87.5369 179.9997 174.6913>
  }
 ]
 upperarm_l [
  {shapeType=Capsule, length=16.0988, radius=7.803, shapeName=upperarm_l_capsule, center=<11.9691 1.0983 0>, rotation=<rot 90 0 0>}
  {
   shapeType=Capsule
   length=8.8415
   radius=5.5494
   shapeName=upperarm_l_capsule
   center=<17.0229 -3.2258 -10.4139>
   rotation=<rot -81.317 -88.4893 88.4718>
  }
 ]
 upperarm_r [{
   shapeType=Capsule
   length=34.3186
   radius=9.0172
   shapeName=upperarm_r_capsule
   center=<-29.505 -5.4739 -0.12>
   rotation=<rot 0 -74.8503 -87.8065>
  }]
}
constraints [
 {bone1Name=dyn_cloak_04, bone2Name=dyn_cloak_03, swing1LimitDegrees=20, swing2LimitDegrees=18, twistLimitDegrees=10}
 {bone1Name=dyn_cloak_01, bone2Name=MSTR_cloak, swing1LimitDegrees=30, swing2LimitDegrees=10, twistLimitDegrees=0}
 {bone1Name=dyn_cloak_02, bone2Name=dyn_cloak_01, swing1LimitDegrees=28, swing2LimitDegrees=14, twistLimitDegrees=0}
 {bone1Name=dyn_cloak_03, bone2Name=dyn_cloak_02, swing1LimitDegrees=30, swing2LimitDegrees=16, twistLimitDegrees=10}
]
