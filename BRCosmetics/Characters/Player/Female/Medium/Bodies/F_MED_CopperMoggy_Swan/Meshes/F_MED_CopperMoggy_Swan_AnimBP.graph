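# Post-process style AnimBlueprint for the CopperMoggy Swan female character body, animating its cloak, sleeve and hose.
# It copies the parent pose, layers in head-mesh jaw bones, then runs a Control Rig driven by skydiving and weapon states.
# Backpacks switch the cloak from front-end or in-game RigidBody simulation to weighted constraints with AnimDynamics.
# Sleeve and hose AnimDynamics follow, then additive bone offsets and a correction Control Rig.
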AnimBlueprint /BRCosmetics/Characters/Player/Female/Medium/Bodies/F_MED_CopperMoggy_Swan/Meshes/F_MED_CopperMoggy_Swan_AnimBP
target_skeleton=&/F_MED_CopperMoggy_Swan_Skeleton
graphs [
 AnimGraph {nodes [
   "Transform (Modify) Bone - Bone: MSTR_shoulder_pad"!(ComponentPose, bAlphaBoolEnabled:bool=True, Translation="0,-0.5,0", Rotation="0,0,0", Scale="1,1,1") -> (Pose) {
    alphaInputType=Bool
    rotationMode=BMM_Additive
    translationMode=BMM_Additive
    alphaBoolBlend {blendInTime=0.1, blendOption=Cubic, blendOutTime=0.8}
    alphaScaleBiasClamp {bInterpResult=true, interpSpeedDecreasing=7, interpSpeedIncreasing=4}
    boneToModify {boneName=MSTR_shoulder_pad}
    translation {y=-0.5}
   }
   n1 = "Transform (Modify) Bone - Bone: MSTR_cloak"!(
    ComponentPose=Local To Component(LocalPose=Use cached pose 'Dynamics'())
    bAlphaBoolEnabled:bool=True
    Translation="0,0,4"
    Rotation="0,0,5"
    Scale="1,1,1"
   ) -> (Pose) {
    alphaInputType=Bool
    rotationMode=BMM_Additive
    translationMode=BMM_Additive
    alphaBoolBlend {blendInTime=0.4, blendOption=Cubic, blendOutTime=0.8}
    alphaScaleBiasClamp {bInterpResult=true, interpSpeedDecreasing=7, interpSpeedIncreasing=4}
    boneToModify {boneName=MSTR_cloak}
    rotation {roll=5}
    translation {z=4}
   }
   n2 = "Transform (Modify) Bone - Bone: MSTR_cloak"(ComponentPose=@n1, bAlphaBoolEnabled=True, Translation="0,-2,2", Rotation="0,0,5", Scale="1,1,1") {
    alphaInputType=Bool
    rotationMode=BMM_Additive
    translationMode=BMM_Additive
    alphaBoolBlend {blendInTime=0.4, blendOption=Cubic, blendOutTime=0.8}
    alphaScaleBiasClamp {bInterpResult=true, interpSpeedDecreasing=7, interpSpeedIncreasing=4}
    boneToModify {boneName=MSTR_cloak}
    rotation {roll=5}
    translation {y=-2, z=2}
   }
   n3 = "Transform (Modify) Bone - Bone: MSTR_cloak"(ComponentPose=@n2, bAlphaBoolEnabled=True, Translation="0,-2,2.5", Rotation="5,0,-10", Scale="1,1,1") {
    alphaInputType=Bool
    rotationMode=BMM_Additive
    translationMode=BMM_Additive
    alphaBoolBlend {blendInTime=0.4, blendOption=Cubic, blendOutTime=0.8}
    alphaScaleBiasClamp {bInterpResult=true, interpSpeedDecreasing=7, interpSpeedIncreasing=4}
    boneToModify {boneName=MSTR_cloak}
    rotation {pitch=5, roll=-10}
    translation {y=-2, z=2.5}
   }
   n4 = "Transform (Modify) Bone - Bone: MSTR_cloak"(ComponentPose=@n3, bAlphaBoolEnabled=True, Translation="0,0,4", Rotation="5,0,-10", Scale="1,1,1") {
    alphaInputType=Bool
    rotationMode=BMM_Additive
    translationMode=BMM_Additive
    alphaBoolBlend {blendInTime=0.4, blendOption=Cubic, blendOutTime=0.8}
    alphaScaleBiasClamp {bInterpResult=true, interpSpeedDecreasing=7, interpSpeedIncreasing=4}
    boneToModify {boneName=MSTR_cloak}
    rotation {pitch=5, roll=-10}
    translation {z=4}
   }
   n5 = "Transform (Modify) Bone - Bone: MSTR_cloak"(ComponentPose=@n4, bAlphaBoolEnabled=True, Translation="0,-1,0", Rotation="0,0,0", Scale="1,1,1") {
    alphaInputType=Bool
    rotationMode=BMM_Additive
    translationMode=BMM_Additive
    alphaBoolBlend {blendInTime=0.4, blendOption=Cubic, blendOutTime=0.8}
    alphaScaleBiasClamp {bInterpResult=true, interpSpeedDecreasing=7, interpSpeedIncreasing=4}
    boneToModify {boneName=MSTR_cloak}
    translation {y=-1}
   }
   n6 = "Transform (Modify) Bone - Bone: MSTR_cloak"(ComponentPose=@n5, bAlphaBoolEnabled=True, Translation="0,-1,0", Rotation="0,0,5", Scale="1,1,1") {
    alphaInputType=Bool
    rotationMode=BMM_Additive
    translationMode=BMM_Additive
    alphaBoolBlend {blendInTime=0.4, blendOption=Cubic, blendOutTime=0.8}
    alphaScaleBiasClamp {bInterpResult=true, interpSpeedDecreasing=7, interpSpeedIncreasing=4}
    boneToModify {boneName=MSTR_cloak}
    rotation {roll=5}
    translation {y=-1}
   }
   n7 = "Transform (Modify) Bone - Bone: MSTR_cloak"(ComponentPose=@n6, bAlphaBoolEnabled=True, Translation="0,-1,0", Rotation="0,0,40", Scale="1,1,1") {
    alphaInputType=Bool
    rotationMode=BMM_Additive
    translationMode=BMM_Additive
    alphaBoolBlend {blendInTime=0.4, blendOption=Cubic, blendOutTime=0.8}
    alphaScaleBiasClamp {bInterpResult=true, interpSpeedDecreasing=7, interpSpeedIncreasing=4}
    boneToModify {boneName=MSTR_cloak}
    rotation {roll=40}
    translation {y=-1}
   }
   n8 = "Transform (Modify) Bone - Bone: MSTR_cloak_correction_l"!(ComponentPose=@n7, bAlphaBoolEnabled:bool=True, Translation="0,-1,1", Rotation="0,0,0", Scale="1,1,1") -> (Pose) {
    alphaInputType=Bool
    rotationMode=BMM_Additive
    translationMode=BMM_Additive
    alphaBoolBlend {blendInTime=0.4, blendOption=Cubic, blendOutTime=0.8}
    alphaScaleBiasClamp {bInterpResult=true, interpSpeedDecreasing=7, interpSpeedIncreasing=4}
    boneToModify {boneName=MSTR_cloak_correction_l}
    translation {y=-1, z=1}
   }
   n9 = "Transform (Modify) Bone - Bone: MSTR_cloak_correction_r"!(ComponentPose=@n8, bAlphaBoolEnabled:bool=True, Translation="0,-1,5", Rotation="0,0,0", Scale="1,1,1") -> (Pose) {
    alphaInputType=Bool
    rotationMode=BMM_Additive
    translationMode=BMM_Additive
    alphaBoolBlend {blendInTime=0.6, blendOption=Cubic, blendOutTime=0.8}
    alphaScaleBiasClamp {bInterpResult=true, interpSpeedDecreasing=7, interpSpeedIncreasing=4}
    boneToModify {boneName=MSTR_cloak_correction_r}
    translation {y=-1, z=5}
   }
   n10 = "Transform (Modify) Bone - Bone: MSTR_cloak_correction_r"(ComponentPose=@n9, bAlphaBoolEnabled=True, Translation="0,-1,0", Rotation="0,0,0", Scale="1,1,1") {
    alphaInputType=Bool
    rotationMode=BMM_Additive
    translationMode=BMM_Additive
    alphaBoolBlend {blendInTime=0.1, blendOption=Cubic, blendOutTime=0.8}
    alphaScaleBiasClamp {bInterpResult=true, interpSpeedDecreasing=7, interpSpeedIncreasing=4}
    boneToModify {boneName=MSTR_cloak_correction_r}
    translation {y=-1}
   }
   n11 = "Transform (Modify) Bone - Bone: MSTR_shoulder_pad"(ComponentPose=@n10, bAlphaBoolEnabled=True, Translation="0,0,1", Rotation="0,0,0", Scale="1,1,1") {
    alphaInputType=Bool
    rotationMode=BMM_Additive
    translationMode=BMM_Additive
    alphaBoolBlend {blendInTime=0.6, blendOption=Cubic, blendOutTime=0.8}
    alphaScaleBiasClamp {bInterpResult=true, interpSpeedDecreasing=7, interpSpeedIncreasing=4}
    boneToModify {boneName=MSTR_shoulder_pad}
    translation {z=1}
   }
   n12 = "Transform (Modify) Bone - Bone: MSTR_cloak_correction_r"!(ComponentPose=@n11, Alpha:float=1, Translation="11,-0.02,0", Rotation="-66.290,0.012,-0.005", Scale="1,1,1") -> (Pose) {
    rotationMode=BMM_Additive
    rotationSpace=BCS_BoneSpace
    translationMode=BMM_Additive
    translationSpace=BCS_BoneSpace
    alphaBoolBlend {blendInTime=0.4, blendOption=Cubic, blendOutTime=0.8}
    alphaScaleBiasClamp {bInterpResult=true, interpSpeedDecreasing=7}
    boneToModify {boneName=MSTR_cloak_correction_r}
    rotation=<rot -66.29 0.012 -0.005>
    translation=<11 -0.02>
   }
   n13 = "Transform (Modify) Bone - Bone: MSTR_belt"(ComponentPose=@n12, Alpha=1, Translation="-2.321,0.948,1.961", Rotation="-0.002,6.902,0.002", Scale="1,1,1") {
    rotationMode=BMM_Additive
    rotationSpace=BCS_BoneSpace
    translationMode=BMM_Additive
    translationSpace=BCS_BoneSpace
    alphaBoolBlend {blendInTime=0.4, blendOption=Cubic, blendOutTime=0.8}
    alphaScaleBiasClamp {bInterpResult=true, interpSpeedDecreasing=7}
    boneToModify {boneName=MSTR_belt}
    rotation=<rot -0.002 6.902 0.002>
    translation=<-2.321 0.948 1.961>
   }
   n14 = "Transform (Modify) Bone - Bone: MSTR_sleeve"(ComponentPose=@n13, Alpha=1, Translation="0,0,0", Rotation="0,84,0", Scale="1,1,1") {
    rotationMode=BMM_Additive
    rotationSpace=BCS_BoneSpace
    translationMode=BMM_Additive
    translationSpace=BCS_BoneSpace
    alphaBoolBlend {blendInTime=0.4, blendOption=Cubic, blendOutTime=0.8}
    alphaScaleBiasClamp {bInterpResult=true, interpSpeedDecreasing=7}
    boneToModify {boneName=MSTR_sleeve}
    rotation {yaw=84}
   }
   n15 = "Transform (Modify) Bone - Bone: MSTR_cloak"!(ComponentPose=@n14, Alpha:float=1, Translation="0,-1.5,0", Rotation="0,0,7", Scale="1,1,1") -> (Pose) {
    rotationMode=BMM_Additive
    rotationSpace=BCS_BoneSpace
    translationMode=BMM_Additive
    translationSpace=BCS_BoneSpace
    alphaBoolBlend {blendInTime=0.4, blendOption=Cubic, blendOutTime=0.8}
    alphaScaleBiasClamp {bInterpResult=true, interpSpeedDecreasing=7}
    boneToModify {boneName=MSTR_cloak}
    rotation {roll=7}
    translation {y=-1.5}
   }
   n16 = "Transform (Modify) Bone - Bone: MSTR_cloak_correction_r"(ComponentPose=@n15, Alpha=1, Translation="0,2,0", Rotation="4,-0.111,0", Scale="1,1,1") {
    rotationMode=BMM_Additive
    rotationSpace=BCS_BoneSpace
    translationMode=BMM_Additive
    translationSpace=BCS_BoneSpace
    alphaBoolBlend {blendInTime=0.4, blendOption=Cubic, blendOutTime=0.8}
    alphaScaleBiasClamp {bInterpResult=true, interpSpeedDecreasing=7}
    boneToModify {boneName=MSTR_cloak_correction_r}
    rotation {pitch=4, yaw=-0.111}
    translation {y=2}
   }
   n17 = Control Rig!(Source=Component To Local(ComponentPose=@n16), Alpha, sky diving:bool=Get bIsSkyDiving()) -> (Pose) {controlRigAssetReference {blueprintRigClass=&/F_MED_CopperMoggy_Swan_Correction_CtrlRig.F_MED_CopperMoggy_Swan_Correction_CtrlRig_C}}
   Output Pose(Result=@n17) {layerGroup=None}
   n18 = Layered blend per bone(
    BasePose=Copy Pose From Mesh() {bCopyCurves=true, bUseAttachedParent=true}
    BlendPoses_0=Copy Pose From Mesh(SourceMeshComponent=Get Body_Mesh()) {bUseAttachedParent=true}
   ) {layerSetup [{branchFilters [
       {boneName=NoneFACIAL_C_Jaw}
       {boneName=FACIAL_C_Jawline}
       {boneName=FACIAL_L_Jawline}
       {boneName=FACIAL_R_Jawline}
       {boneName=FACIAL_C_UnderChin}
       {boneName=FACIAL_L_UnderChin}
       {boneName=FACIAL_R_UnderChin}
      ]}]}
   Save cached pose 'StartCache'(Pose=@n18) {cacheName=StartCache}
   n19 = Anim Dynamics - Bone: dyn_sleeve(ComponentPose=Local To Component(LocalPose=Use cached pose 'Rban'()), Alpha=1, ExternalForce="0,0,0") {
    angularDampingOverride=0.999
    angularSpringConstant=2500
    bAngularSpring=true
    bOverrideAngularDamping=true
    bUseGravityOverride=true
    gravityScale=0.35
    boundBone {boneName=dyn_sleeve}
    componentAppliedLinearAccClamp=<3000 3000 3000>
    componentLinearAccScale=<1 1 1>
    componentLinearVelScale=<1 1 1>
    gravityOverride {z=-980}
    physicsBodyDefinitions [{
      boundBone {boneName=dyn_sleeve}
      boxExtents=<20 20 20>
      constraintSetup {angularLimitsMax=<15 15 75>, angularLimitsMin=<-15 -15 -75>, angularTarget=<1 0 0>}
      localJointOffset=<2 2 0>
     }]
    relativeSpaceBone {boneName=upperarm_l}
    simSpaceSettings {maxAngularAcceleration=2500, maxAngularVelocity=2500, simSpaceAngularAlpha=1}
   }
   n20 = Anim Dynamics - Bone: dyn_sleeve(ComponentPose=@n19, Alpha=1, ExternalForce="0,0,0") {
    angularDampingOverride=0.999
    angularSpringConstant=2500
    bAngularSpring=true
    bOverrideAngularDamping=true
    bUseGravityOverride=true
    gravityScale=0.35
    boundBone {boneName=dyn_sleeve}
    componentAppliedLinearAccClamp=<3000 3000 3000>
    gravityOverride {z=-650}
    physicsBodyDefinitions [{
      boundBone {boneName=dyn_sleeve}
      boxExtents=<27 27 27>
      constraintSetup {angularLimitsMax=<15 5 75>, angularLimitsMin=<-15 -5 -15>, angularTarget=<1 0 0>}
      localJointOffset=<0 0 0>
     }]
    relativeSpaceBone {boneName=upperarm_l}
    simSpaceSettings {maxAngularAcceleration=56, maxAngularVelocity=8, simSpaceAngularAlpha=1}
    sphericalLimits [{limitRadius=10, drivingBone {boneName=dyn_cloak_01}, sphereLocalOffset=<10 -12 15>}]
   }
   n21 = Anim Dynamics - Bone: dyn_hose(ComponentPose=@n20, Alpha=1, ExternalForce="0,0,0") {
    angularDampingOverride=1
    angularSpringConstant=4500
    bAngularSpring=true
    bOverrideAngularDamping=true
    bUseGravityOverride=true
    gravityScale=0.35
    simulationSpace=RootRelative
    boundBone {boneName=dyn_hose}
    componentLinearAccScale=<1 1 1>
    componentLinearVelScale=<1 1 1>
    gravityOverride {z=-980}
    physicsBodyDefinitions [{
      boundBone {boneName=dyn_hose}
      boxExtents=<20 20 20>
      constraintSetup {angularLimitsMax=<0 5 0>, angularLimitsMin=<0 -20 0>, angularTarget=<1 0 0>}
      localJointOffset=<3 0 0>
     }]
    simSpaceSettings {maxAngularAcceleration=35, maxAngularVelocity=6, simSpaceAngularAlpha=1}
   }
   Save cached pose 'Dynamics'(Pose=Component To Local(ComponentPose=@n21)) {cacheName=Dynamics}
   n22 = Control Rig!(
    Source=Use cached pose 'StartCache'()
    Alpha
    skydiving:bool=Get bIsSkyDiving()
    parachute opened:bool=Get bIsParachuteOpened()
    targeting:double=Get isTargetingAlpha() -> (isTargetingAlpha:double)
    Is pistol alpha:double=Get Is Pistol Alpha() -> (Is Pistol Alpha:float)
   ) -> (Pose) {controlRigAssetReference {blueprintRigClass=&/F_MED_CopperMoggy_Swan_CtrlRig.F_MED_CopperMoggy_Swan_CtrlRig_C}}
   Save cached pose 'Control Rig'(Pose=@n22) {cacheName=Control Rig}
   n23 = RigidBody!(
    ComponentPose=Local To Component(LocalPose=Use cached pose 'PreDynamicsAdjustments'())
    Alpha
    OverrideWorldGravity:Vector="0,0,-980"
    ExternalForce
   ) -> (Pose) {
    bClampLinearTranslationLimitToRefPose=true
    bOverrideWorldGravity=true
    overridePhysicsAsset=&/F_MED_CopperMoggy_Swan_FrontEnd_Physics
    simulationSpace=BaseBoneSpace
    baseBoneRef {boneName=spine_05}
    componentAppliedLinearAccClamp=<3000 3000 3000>
    simSpaceSettings {
     dampingAlpha=0
     maxAngularAcceleration=40
     maxAngularVelocity=8
     maxLinearAcceleration=2800
     maxLinearVelocity=700
     velocityScaleZ=0.5
     worldAlpha=1
     externalLinearDragV=<3 0.3 0.75>
    }
   }
   n24 = RigidBody(ComponentPose=Local To Component(LocalPose=Use cached pose 'PreDynamicsAdjustments'()), OverrideWorldGravity="0,0,-980") {
    bClampLinearTranslationLimitToRefPose=true
    bOverrideWorldGravity=true
    overridePhysicsAsset=&/F_MED_CopperMoggy_Swan_Physics
    simulationSpace=BaseBoneSpace
    baseBoneRef {boneName=spine_05}
    componentAppliedLinearAccClamp=<3000 3000 3000>
    simSpaceSettings {
     dampingAlpha=0
     maxAngularAcceleration=40
     maxAngularVelocity=8
     maxLinearAcceleration=2800
     maxLinearVelocity=700
     velocityScaleZ=0.5
     worldAlpha=1
     externalLinearDragV=<3 0.3 0.75>
    }
   }
   n25 = "Constraint - dyn_cloak_01 (PPPP)"!(
    ComponentPose=Local To Component(LocalPose=Use cached pose 'Control Rig'())
    Alpha=1
    ConstraintWeights_0=0.5979
    ConstraintWeights_1:float=0.195
    ConstraintWeights_2:float=0.195
    ConstraintWeights_3:float=0.0121
   ) -> (Pose) {
    boneToModify {boneName=dyn_cloak_01}
    constraintSetup [{targetBone {boneName=spine_05}}, {targetBone {boneName=clavicle_l}}, {targetBone {boneName=clavicle_r}}, {targetBone {boneName=spine_04}}]
    constraintWeights [1,1,1,1]
   }
   n26 = "Constraint - dyn_cloak_02 (PPPPPPP)"!(
    ComponentPose=@n25
    Alpha=1
    ConstraintWeights_0=0.3529
    ConstraintWeights_1:float=0.3436
    ConstraintWeights_2:float=0.101
    ConstraintWeights_3:float=0.0904
    ConstraintWeights_4:float=0.0506
    ConstraintWeights_5:float=0.0474
    ConstraintWeights_6:float=0.014
   ) -> (Pose) {
    boneToModify {boneName=dyn_cloak_02}
    constraintSetup [
     {targetBone {boneName=spine_05}}
     {targetBone {boneName=spine_04}}
     {targetBone {boneName=clavicle_l}}
     {targetBone {boneName=clavicle_r}}
     {targetBone {boneName=spine_03}}
     {targetBone {boneName=spine_02}}
     {targetBone {boneName=spine_01}}
    ]
    constraintWeights [1,1,1,1,1,1,1]
   }
   n27 = "Constraint - dyn_cloak_03 (PPPPPPPP)"!(
    ComponentPose=@n26
    Alpha=1
    ConstraintWeights_0=0.4548
    ConstraintWeights_1:float=0.1889
    ConstraintWeights_2:float=0.1227
    ConstraintWeights_3:float=0.0908
    ConstraintWeights_4:float=0.0504
    ConstraintWeights_5:float=0.0462
    ConstraintWeights_6:float=0.0287
    ConstraintWeights_7:float=0.0175
   ) -> (Pose) {
    boneToModify {boneName=dyn_cloak_03}
    constraintSetup [
     {targetBone {boneName=spine_04}}
     {targetBone {boneName=spine_02}}
     {targetBone {boneName=spine_03}}
     {targetBone {boneName=spine_05}}
     {targetBone {boneName=spine_01}}
     {targetBone {boneName=clavicle_l}}
     {targetBone {boneName=clavicle_r}}
     {targetBone {boneName=pelvis}}
    ]
    constraintWeights [1,1,1,1,1,1,1,1]
   }
   n28 = Anim Dynamics - Bone: dyn_cloak_04(ComponentPose=@n27, Alpha=1, ExternalForce="0,0,0") {
    angularDampingOverride=0.9999
    angularSpringConstant=2500
    bAngularSpring=true
    bOverrideAngularDamping=true
    bUseGravityOverride=true
    bUseSphericalLimits=true
    gravityScale=0.35
    boundBone {boneName=dyn_cloak_04}
    componentLinearAccScale=<1 1 1>
    componentLinearVelScale=<1 1 1>
    gravityOverride {z=-980}
    physicsBodyDefinitions [{
      collisionType=InnerSphere
      boundBone {boneName=dyn_cloak_04}
      boxExtents=<22 20 30>
      constraintSetup {angularLimitsMax=<15 20 15>, angularLimitsMin=<-15 -20 -50>, angularTarget=<1 0 0>}
      localJointOffset=<1 0 1.4>
     }]
    simSpaceSettings {maxAngularAcceleration=35, maxAngularVelocity=6, simSpaceAngularAlpha=1, externalAngularVelocity=<1 1 1>}
    sphericalLimits [
     {limitRadius=6, drivingBone {boneName=pelvis}, sphereLocalOffset=<0 0 0>}
     {limitRadius=5, drivingBone {boneName=pelvis}, sphereLocalOffset=<8 -4 10>}
    ]
   }
   n29 = "Transform (Modify) Bone - Bone: MSTR_cloak"(ComponentPose=@n28, Alpha=1, Translation="0,0,0", Rotation="0,10,0", Scale="1,1,1") {
    rotationMode=BMM_Additive
    rotationSpace=BCS_BoneSpace
    translationMode=BMM_Additive
    translationSpace=BCS_BoneSpace
    alphaBoolBlend {blendInTime=0.4, blendOption=Cubic, blendOutTime=0.8}
    alphaScaleBiasClamp {bInterpResult=true, interpSpeedDecreasing=7}
    boneToModify {boneName=MSTR_cloak}
    rotation {yaw=10}
   }
   n30 = "Transform (Modify) Bone - Bone: MSTR_cloak"(ComponentPose=@n29, Alpha=1, Translation="0,-4,0", Rotation="0,7,0", Scale="1,1,1") {
    rotationMode=BMM_Additive
    rotationSpace=BCS_BoneSpace
    translationMode=BMM_Additive
    translationSpace=BCS_BoneSpace
    alphaBoolBlend {blendInTime=0.4, blendOption=Cubic, blendOutTime=0.8}
    alphaScaleBiasClamp {bInterpResult=true, interpSpeedDecreasing=7}
    boneToModify {boneName=MSTR_cloak}
    rotation {yaw=7}
    translation {y=-4}
   }
   n31 = "Transform (Modify) Bone - Bone: dyn_cloak_03"(ComponentPose=@n30, Alpha=1, Translation="0,0,0", Rotation="0,4,0", Scale="1,1,1") {
    rotationMode=BMM_Additive
    rotationSpace=BCS_BoneSpace
    translationMode=BMM_Additive
    translationSpace=BCS_BoneSpace
    alphaBoolBlend {blendInTime=0.4, blendOption=Cubic, blendOutTime=0.8}
    alphaScaleBiasClamp {bInterpResult=true, interpSpeedDecreasing=7, interpSpeedIncreasing=7}
    boneToModify {boneName=dyn_cloak_03}
    rotation {yaw=4}
   }
   n32 = "Transform (Modify) Bone - Bone: MSTR_cloak"(ComponentPose=@n31, Alpha=1, Translation="0,0,0", Rotation="0,-10,0", Scale="1,1,1") {
    rotationMode=BMM_Additive
    rotationSpace=BCS_BoneSpace
    translationMode=BMM_Additive
    translationSpace=BCS_BoneSpace
    alphaBoolBlend {blendInTime=0.4, blendOption=Cubic, blendOutTime=0.8}
    alphaScaleBiasClamp {bInterpResult=true, interpSpeedDecreasing=8}
    boneToModify {boneName=MSTR_cloak}
    rotation {yaw=-10}
   }
   n33 = Anim Dynamics - Bone: dyn_cloak_04(ComponentPose=@n32, Alpha=1, ExternalForce="0,0,0") {
    angularDampingOverride=0.9999
    angularSpringConstant=2500
    bAngularSpring=true
    bOverrideAngularDamping=true
    bUseGravityOverride=true
    bUseSphericalLimits=true
    gravityScale=0.35
    boundBone {boneName=dyn_cloak_04}
    componentLinearAccScale=<1 1 1>
    componentLinearVelScale=<1 1 1>
    gravityOverride {z=-980}
    physicsBodyDefinitions [{
      collisionType=CustomSphere
      sphereCollisionRadius=6.5
      boundBone {boneName=dyn_cloak_04}
      boxExtents=<22 20 30>
      constraintSetup {angularLimitsMax=<15 20 15>, angularLimitsMin=<-15 -20 -50>, angularTarget=<1 0 0>}
      localJointOffset=<1 0 1.4>
     }]
    simSpaceSettings {maxAngularAcceleration=35, maxAngularVelocity=6, simSpaceAngularAlpha=1, externalAngularVelocity=<1 1 1>}
    sphericalLimits [
     {limitRadius=6, drivingBone {boneName=pelvis}, sphereLocalOffset=<0 0 0>}
     {limitRadius=5, drivingBone {boneName=pelvis}, sphereLocalOffset=<8 -4 10>}
    ]
   }
   n34 = "Constraint - dyn_cloak_04 (PPP)"!(ComponentPose=@n33, Alpha=1, ConstraintWeights_0=0.9269, ConstraintWeights_1:float=0.0507, ConstraintWeights_2:float=0.0225) -> (Pose) {
    boneToModify {boneName=dyn_cloak_04}
    constraintSetup [{targetBone {boneName=pelvis}}, {targetBone {boneName=deform_glute_l}}, {targetBone {boneName=spine_02}}]
    constraintWeights [1,1,1]
   }
   n35 = "Transform (Modify) Bone - Bone: dyn_cloak_04"(ComponentPose=@n34, Alpha=1, Translation="0,0,0", Rotation="0,-15,0", Scale="1,1,1") {
    rotationMode=BMM_Additive
    rotationSpace=BCS_BoneSpace
    translationMode=BMM_Additive
    translationSpace=BCS_BoneSpace
    alphaBoolBlend {blendInTime=0.4, blendOption=Cubic, blendOutTime=0.8}
    alphaScaleBiasClamp {bInterpResult=true, interpSpeedDecreasing=8}
    boneToModify {boneName=dyn_cloak_04}
    rotation {yaw=-15}
   }
   n36 = "Transform (Modify) Bone - Bone: dyn_cloak_02"(ComponentPose=@n35, Alpha=1, Translation="0,0,0", Rotation="0,-9,0", Scale="1,1,1") {
    rotationMode=BMM_Additive
    rotationSpace=BCS_BoneSpace
    translationMode=BMM_Additive
    translationSpace=BCS_BoneSpace
    alphaBoolBlend {blendInTime=0.4, blendOption=Cubic, blendOutTime=0.8}
    alphaScaleBiasClamp {bInterpResult=true, interpSpeedDecreasing=8}
    boneToModify {boneName=dyn_cloak_02}
    rotation {yaw=-9}
   }
   n37 = Blend Poses by bool(
    BlendPose_0=Component To Local(ComponentPose=@n36)
    BlendPose_1=Blend Poses by bool(BlendPose_0=Component To Local(ComponentPose=@n23), BlendPose_1=Component To Local(ComponentPose=@n24), bActiveValue=Get bIsInFrontEnd()) {blendType=Linear}
    bActiveValue=Get IsWearingBackpack() -> (IsWearingBackpack:bool)
   ) {blendType=Linear}
   Save cached pose 'Rban'(Pose=Component To Local(ComponentPose=Local To Component(LocalPose=@n37))) {cacheName=Rban}
   n38 = "Transform (Modify) Bone - Bone: MSTR_cloak"(
    ComponentPose=Local To Component(LocalPose=Use cached pose 'Control Rig'())
    Alpha=1
    Translation="-10,0,0"
    Rotation="0,30,0"
    Scale="1,1,1"
   ) {
    rotationMode=BMM_Additive
    rotationSpace=BCS_BoneSpace
    translationMode=BMM_Additive
    translationSpace=BCS_BoneSpace
    alphaBoolBlend {blendInTime=0.4, blendOption=Cubic, blendOutTime=0.8}
    alphaScaleBiasClamp {bInterpResult=true, interpSpeedDecreasing=7}
    boneToModify {boneName=MSTR_cloak}
    rotation {yaw=30}
    translation {x=-10}
   }
   n39 = "Transform (Modify) Bone - Bone: upperarm_l"(ComponentPose=@n38, Alpha=1, Translation="0,0,0", Rotation="0,0,0", Scale="1,1,1") {
    rotationMode=BMM_Additive
    rotationSpace=BCS_BoneSpace
    alphaScaleBiasClamp {bInterpResult=true, interpSpeedDecreasing=5, interpSpeedIncreasing=5}
    boneToModify {boneName=upperarm_l}
   }
   Save cached pose 'PreDynamicsAdjustments'(Pose=Component To Local(ComponentPose=@n39)) {cacheName=PreDynamicsAdjustments}
  ]}
 EventGraph {nodes [
   Event BlueprintUpdateAnimation()
   > n1 = Sequence!(execute) -> (then_0, then_1, then_2:Exec, then_3:Exec, then_4:Exec)
   > n2 = "Gravity Override State Function (Parameter Asset) (Nativized)"(CharacterPartAnimInstance=Self-Reference(), GameTimeInSeconds=GetGameTimeInSeconds())
   > Set GravityOverride(GravityOverride=@n2)
   n3 = Sequence!(execute=@n1.then_2) -> (then_0, then_1, then_2)
   > Set MovingAndCrouching(MovingAndCrouching:bool=AND Boolean(A=Get bIsMoving2D(), B=Get bIsCrouching())) -> (Output_Get:bool)
   n4 = Get IsWearingBackpack()
   > Set PlayingMeleeWithoutBackpack(PlayingMeleeWithoutBackpack:bool=AND Boolean(A=Get bIsPlayingMeleeAnim(), B=NOT Boolean(A=@n4))) -> (Output_Get:bool)
   > Set isTargetingAlpha(isTargetingAlpha:double="To Float (Boolean)"(InBool=Get bIsTargeting())) -> (Output_Get:double)
   > Set WearingBackpackInFrontend(WearingBackpackInFrontend:bool=AND Boolean(A=Get IsWearingBackpack(), B=Get bIsInFrontEnd())) -> (Output_Get:bool)
   n5 = Get Is Wearing Cape() -> (Is Wearing Cape:bool)
   n6 = Get bIsInFrontEnd()
   Set AlphaCoatDynamicsWearingBackpack(execute=@n3.then_2, AlphaCoatDynamicsWearingBackpack:bool=AND Boolean(A=NOT Boolean(A=@n5), B=NOT Boolean(A=@n6))) -> (Output_Get:bool)
   > Set CrouchingWithCape(CrouchingWithCape:bool=AND Boolean(A=@n5, B=Get bIsCrouching())) -> (Output_Get:bool)
   > Set RunningWithCape(RunningWithCape:bool=AND Boolean(A=Get Is Wearing Cape(), B=Get bIsTacticalSprinting())) -> (Output_Get:bool)
   > Set NotFrontEnd(NotFrontEnd:bool=NOT Boolean(A=Get bIsInFrontEnd())) -> (Output_Get:bool)
   n7 = Get IsWearingBackpack()
   Set skyDiveWithoutBackpack(execute=@n3.then_1, skyDiveWithoutBackpack:bool=AND Boolean(A=Get bIsSkydiveDiveMode(), B=NOT Boolean(A=@n7))) -> (Output_Get:bool)
   > Set PlayingMeleeCrouching(PlayingMeleeCrouching:bool=AND Boolean(A=Get bIsPlayingMeleeAnim(), B=Get bIsCrouching())) -> (Output_Get:bool)
   n8 = Sequence(execute=@n1.then_3)
   > n9 = Cast To FortPlayerAnimInstance(Object=GetAnimInstance(self=GetOwnerSkeletalMeshComponent()))
   > Set isPistol(isPistol=false)
   > n10 = Switch on EFortWeaponCoreAnimation!(execute, Selection:EFortWeaponCoreAnimation Enum=Get WeaponCoreAnim(self=@n9)) -> (
    Melee:Exec
    Pistol:Exec
    Shotgun:Exec
    PaperBlueprint:Exec
    Rifle:Exec
    MeleeOneHand:Exec
    MachinePistol:Exec
    RocketLauncher:Exec
    GrenadeLauncher:Exec
    GoingCommando:Exec
    AssaultRifle:Exec
    TacticalShotgun:Exec
    SniperRifle:Exec
    TrapPlacement:Exec
    ShoulderLauncher:Exec
    AbilityDecoTool:Exec
    Crossbow:Exec
    C4:Exec
    RemoteControl:Exec
    DualWield:Exec
    AR_BullPup:Exec
    AR_ForwardGrip:Exec
    MedPackPaddles:Exec
    SMG_P90:Exec
    AR_DrumGun:Exec
    Consumable_Small:Exec
    Consumable_Large:Exec
    Balloon:Exec
    MountedTurret:Exec
    CreativeTool:Exec
    ExplosiveBow:Exec
    AshtonIndigo:Exec
    AshtonChicago:Exec
    MeleeDualWield:Exec
    Unarmed:Exec
   )
   Set isPistol(execute=@n10.Pistol, isPistol=true)
   Sequence(execute=@n8.then_1)
   n11 = Get isPistol() -> (isPistol:bool)
   n12 = Get bIsTargeting()
   > Set Is Pistol Targeting Alpha(Is Pistol Targeting Alpha:float="To Float (Boolean)"(InBool=AND Boolean(A=@n11, B=@n12))) -> (Output_Get:float)
   > Set Is Pistol Alpha(Is Pistol Alpha:float="To Float (Boolean)"(InBool=Get isPistol())) -> (Output_Get:float)
   n13 = GetOwnerSkeletalMeshComponent()
   n14 = BreakRotator(InRot=BreakTransform(InTransform=GetDeltaTransformFromRefPose(self=@n13, BoneName=upperarm_l, BaseName=clavicle_l)).Rotation)
   n15 = Get Y percentage() -> (Y percentage:float)
   n16 = Get Z percentage() -> (Z percentage:float)
   n17 = Get X percentage() -> (X percentage:float)
   Set Upperarm Rotation(
    execute=@n8.then_2
    Upperarm Rotation:Rotator=MakeRotator(
     Roll=float * float!(A=@n14, B=-1, C:double=@n17) -> (ReturnValue)
     Pitch=float * float(A=@n14.Pitch, B=-1, C=@n15)
     Yaw=float * float(A=@n14.Yaw, B=-1, C=@n16)
    )
   ) -> (Output_Get:Rotator)
   n18 = Set Alpha Output(
    execute=@n1.then_4
    Alpha Output:float=GetAlphaFromReferenceTransformDelta(SkeletalMeshComponent=GetOwnerSkeletalMeshComponent(), BoneName=spine_05, BaseName=pelvis, InputRange=(X=-40,Y=-60), TwistAxis="1,0,0")
   ) -> (Output_Get:float)
   > Set Alpha Spine rotations(Alpha Spine rotations:bool=AND Boolean(A="To Bool (Double)"!(Value:double=@n18) -> (ReturnValue:bool), B=Get bIsParachuteOpened())) -> (Output_Get:bool)
   n19 = CheckForBackbling()
   Set IsWearingBackpack(execute=@n1.then_1, IsWearingBackpack:bool=OR Boolean!(A=@n19, B=@n19.bIsBackblingACape, C:bool=@n19.bIsBackblingLong) -> (ReturnValue)) -> (Output_Get:bool)
   > Set Is Wearing Cape(Is Wearing Cape=OR Boolean!(A=@n19.bIsBackblingACape, B=@n19.bIsBackblingLong) -> (ReturnValue))
   n20 = Event OnCharacterCustomizationLoaded()
   > Set Body_Mesh(Body_Mesh=Get Array Item with EFort Custom Part Type(Array=@n20.CustomMeshParts, PartType=Head))
   > Get Dynamic Material Instance(SkeletalMeshComponent=Get Array Item with EFort Custom Part Type(Array=@n20.CustomMeshParts, PartType=Head))
  ]}
]
variables [
 Alpha Output {type=number}
 Alpha Spine rotations {type=boolean}
 AlphaCoatDynamicsWearingBackpack {type=boolean}
 Body_Mesh {type=SkeletalMeshComponent}
 Crouching with Cape {type=boolean}
 CrouchingWithCape {type=boolean}
 GravityOverride {type=Vector}
 Is Pistol Alpha {type=number}
 Is Pistol Targeting Alpha {type=number}
 Is Wearing Cape {type=boolean}
 isPistol {type=boolean}
 isTargetingAlpha {type=number}
 IsWearingBackpack {type=boolean}
 MovingAndCrouching {type=boolean}
 NotFrontEnd {type=boolean}
 PlayingMeleeCrouching {type=boolean}
 PlayingMeleeWithoutBackpack {type=boolean}
 RunningWithCape {type=boolean}
 skyDiveWithoutBackpack {type=boolean}
 Upperarm Rotation {type=Rotator}
 WearingBackpackInFrontend {type=boolean}
 X percentage {type=number, default=0.2}
 Y percentage {type=number, default=0.2}
 Z percentage {type=number, default=0.2}
]
