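# Physics asset for the CopperMoggy Swan female medium player body, providing, by its name, collision shapes for the character's cape.
# Single capsules cover the pelvis, spine, arms, hands, legs and feet, plus two cloak bones, dyn_cloak_01 and dyn_cloak_03.
# Constraints with uniform swing and twist limits link them in a pelvis-rooted tree, the cloak chaining from spine_05.
# The right upper arm uses its twist bone, so lowerarm_r constrains directly to spine_05.
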PhysicsAsset /BRCosmetics/Characters/Player/Female/Medium/Bodies/F_MED_CopperMoggy_Swan/Meshes/F_MED_CopperMoggy_Swan_CapeColiders_Physics
bodies {
 calf_l [{shapeType=Capsule, length=24.4655, radius=6.8994, center=<-20.5757 -1.8469 -0.6296>, rotation=<rot 0 94.3739 -89.0004>}]
 calf_r [{shapeType=Capsule, length=24.3883, radius=7.8344, center=<20.6967 0.7892 0.6662>, rotation=<rot 0 93.2521 -88.8319>}]
 dyn_cloak_01 [{shapeType=Capsule, length=16.1218, radius=15.2505, center=<13.3405 10.1082 0.4588>, rotation=<rot 3.3062 -88.5109 21.3874>}]
 dyn_cloak_03 [{shapeType=Capsule, length=16.8, radius=11.098, center=<12.5859 6.7858 -0.2901>, rotation=<rot -1.7283 -90.8678 23.7671>}]
 foot_l [{shapeType=Capsule, length=15.9183, radius=9.0954, center=<-3.6369 8.0172 0.1404>, rotation=<rot 0 17.3004 90.4771>}]
 foot_r [{shapeType=Capsule, length=15.9183, radius=9.0954, center=<3.6369 -8.0172 -0.1404>, rotation=<rot 0 17.3004 -89.5229>}]
 hand_l [{shapeType=Capsule, length=9.2469, radius=6.0448, center=<7.5444 -2.2288 0.0667>, rotation=<rot 0 82.2347 -74.5457>}]
 hand_r [{shapeType=Capsule, length=9.2317, radius=5.5921, center=<-7.6877 2.0138 0.2666>, rotation=<rot 0 79.716 -76.4521>}]
 lowerarm_l [{shapeType=Capsule, length=23.2021, radius=7.4809, center=<9.6602 -0.0333 0.4209>, rotation=<rot 0 -88.7898 -84.0202>}]
 lowerarm_r [{shapeType=Capsule, length=14.6228, radius=5.1205, center=<-9.9462 -0.0083 -0.4168>, rotation=<rot 0 -89.0658 -86.7826>}]
 pelvis [{shapeType=Capsule, length=13.713, radius=14.1684, center=<-2.7356 0 0>}]
 spine_02 [{shapeType=Capsule, length=11.0837, radius=9.2667, center=<2.862 0.8545 -0.9309>, rotation=<rot 6.6293 -90 -6.3582>}]
 spine_04 [{shapeType=Capsule, length=13.6605, radius=12.8844, center=<-0.7027 1.1923 -0.7345>, rotation=<rot 0 0.4534 -76.3308>}]
 spine_05 [{shapeType=Capsule, length=20.2744, radius=10.4995}]
 thigh_l [{shapeType=Capsule, length=22.9105, radius=9.8806, center=<-21.564 -1.6893 0.6496>, rotation=<rot 0 -92.2173 -88.2181>}]
 thigh_r [{shapeType=Capsule, length=23.2417, radius=10.5931, center=<21.3368 0.8255 -0.6094>, rotation=<rot 0 -93.2566 -82.3446>}]
 upperarm_l [{shapeType=Capsule, length=19.14, radius=8.3539, center=<10.4413 -0.2301 -2.5927>, rotation=<rot 0 -95.578 -84.5423>}]
 upperarm_twist_01_r [{shapeType=Capsule, length=13.2384, radius=8.5015, center=<0.8548 1.09 0.3074>, rotation=<rot 0 -103.0953 -73.9968>}]
}
constraints [
 {bone1Name=spine_02, bone2Name=pelvis, swing1LimitDegrees=45, swing2LimitDegrees=45, twistLimitDegrees=45}
 {bone1Name=spine_04, bone2Name=spine_02, swing1LimitDegrees=45, swing2LimitDegrees=45, twistLimitDegrees=45}
 {bone1Name=spine_05, bone2Name=spine_04, swing1LimitDegrees=45, swing2LimitDegrees=45, twistLimitDegrees=45}
 {bone1Name=upperarm_l, bone2Name=spine_05, swing1LimitDegrees=45, swing2LimitDegrees=45, twistLimitDegrees=45}
 {bone1Name=lowerarm_l, bone2Name=upperarm_l, swing1LimitDegrees=45, swing2LimitDegrees=45, twistLimitDegrees=45}
 {bone1Name=hand_l, bone2Name=lowerarm_l, swing1LimitDegrees=45, swing2LimitDegrees=45, twistLimitDegrees=45}
 {bone1Name=lowerarm_r, bone2Name=spine_05, swing1LimitDegrees=45, swing2LimitDegrees=45, twistLimitDegrees=45}
 {bone1Name=hand_r, bone2Name=lowerarm_r, swing1LimitDegrees=45, swing2LimitDegrees=45, twistLimitDegrees=45}
 {bone1Name=upperarm_twist_01_r, bone2Name=spine_05, swing1LimitDegrees=45, swing2LimitDegrees=45, twistLimitDegrees=45}
 {bone1Name=dyn_cloak_01, bone2Name=spine_05, swing1LimitDegrees=45, swing2LimitDegrees=45, twistLimitDegrees=45}
 {bone1Name=dyn_cloak_03, bone2Name=dyn_cloak_01, swing1LimitDegrees=45, swing2LimitDegrees=45, twistLimitDegrees=45}
 {bone1Name=thigh_l, bone2Name=pelvis, swing1LimitDegrees=45, swing2LimitDegrees=45, twistLimitDegrees=45}
 {bone1Name=calf_l, bone2Name=thigh_l, swing1LimitDegrees=45, swing2LimitDegrees=45, twistLimitDegrees=45}
 {bone1Name=foot_l, bone2Name=calf_l, swing1LimitDegrees=45, swing2LimitDegrees=45, twistLimitDegrees=45}
 {bone1Name=thigh_r, bone2Name=pelvis, swing1LimitDegrees=45, swing2LimitDegrees=45, twistLimitDegrees=45}
 {bone1Name=calf_r, bone2Name=thigh_r, swing1LimitDegrees=45, swing2LimitDegrees=45, twistLimitDegrees=45}
 {bone1Name=foot_r, bone2Name=calf_r, swing1LimitDegrees=45, swing2LimitDegrees=45, twistLimitDegrees=45}
]
